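# Control Rig for the Love Queen female medium body cosmetic that re-poses the skirt bones, apparently for skydiving.
# The Skydive Float variable passes through AlphaInterp, which eases it at equal rates up and down.
# The smoothed value weights three chained Modify Transforms units that apply fixed rotation offsets to skirt_l_01, skirt_r_01 and skirt_01.
# The graph modifies no other bones in the imported hierarchy directly.
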ControlRigBlueprint /BRCosmetics/Characters/Player/Female/Medium/Bodies/F_MED_Love_Queen/Meshes/F_MED_Love_Queen_CtrlRig
graphs [
 RigVMFunctionLibrary {links [], nodes []}
 RigVMModel {nodes [
   RigUnit_BeginExecution = &RigUnit_BeginExecution()
   AlphaInterp = &RigVMFunction_AlphaInterp(InterpSpeedDecreasing=5, InterpSpeedIncreasing=5, Scale=1, bClampResult=False, Value=RigVMVariableNode() {variable=SKydive Float}.Value)
   &RigUnit_ModifyTransforms(
    ItemToModify=((Item=(Type=Bone,Name="skirt_l_01"),Transform=(Rotation=(X=-0.3444,Y=-0.2817,Z=-0.5098,W=0.7363),Translation=(X=0,Y=0,Z=0),Scale3D=(X=1,Y=1,Z=1))))
    WeightMinimum=0
    ExecutePin=@RigUnit_BeginExecution.ExecutePin
    Weight=@AlphaInterp.Result
   )
   > &RigUnit_ModifyTransforms(
    ItemToModify=((Item=(Type=Bone,Name="skirt_r_01"),Transform=(Rotation=(X=-0.3107,Y=0.3185,Z=-0.3107,W=0.8399),Translation=(X=0,Y=0,Z=0),Scale3D=(X=1,Y=1,Z=1))))
    WeightMinimum=0
    Weight=@AlphaInterp.Result
   )
   > &RigUnit_ModifyTransforms(
    ItemToModify=((Item=(Type=Bone,Name="skirt_01"),Transform=(Rotation=(X=0,Y=0,Z=-0.2588,W=0.9659),Translation=(X=0,Y=0,Z=0),Scale3D=(X=1,Y=1,Z=1))))
    WeightMinimum=0
    Weight=@AlphaInterp.Result
   )
  ]}
]
hierarchy [
 root {loc=<0 0 0>, rot=<0 0 0 1>}
 pelvis {parent=root, loc=<0 -0.5301 98.0571>, rot=<0.0127 0.707 -0.0127 -0.707>}
 spine_01 {parent=pelvis, loc=<6.1208 -0.9328 0>, rot=<0 0 0.034 -0.9994>}
 spine_02 {parent=spine_01, loc=<6.5996 -0.2095 0>, rot=<0 0 0.0238 -0.9997>}
 spine_03 {parent=spine_02, loc=<6.4914 -0.1237 0>, rot=<0 0 0.0149 -0.9999>}
 spine_04 {parent=spine_03, loc=<6.9598 0.0794 0>, rot=<0 0 -0.0132 -0.9999>}
 spine_05 {parent=spine_04, loc=<6.9987 0 0>, rot=<0 0 -0.0015 -1>}
 clavicle_l {parent=spine_05, loc=<5.4831 -0.1053 -3.2946>, rot=<0.0465 -0.7289 -0.009 -0.6829>}
 upperarm_l {parent=clavicle_l, loc=<9.8658 -1.2136 -0.4105>, rot=<0.0164 -0.3245 -0.0269 -0.9454>}
 lowerarm_l {parent=upperarm_l, loc=<26.465 -0.8345 0.259>, rot=<0.0036 0.0161 -0.2298 -0.9731>}
 hand_l {parent=lowerarm_l, loc=<22.7355 -0.3203 -0.6105>, rot=<-0.6784 -0.0223 0.0735 -0.7307>}
 index_metacarpal_l {parent=hand_l, loc=<1.6535 0.0183 -2.96>, rot=<0.0158 -0.0124 -0.0173 -0.9996>}
 index_01_l {parent=index_metacarpal_l, loc=<8.3008 0.0357 -0.0524>, rot=<-0.0044 0.0485 0.0924 -0.9945>}
 index_02_l {parent=index_01_l, loc=<3.3587 -0.3004 0.0452>, rot=<-0.002 0.0027 0.3225 -0.9466>}
 index_03_l {parent=index_02_l, loc=<2.5578 -0.123 -0.0252>, rot=<-0.002 0.0067 0.2558 -0.9667>}
 middle_metacarpal_l {parent=hand_l, loc=<2.1856 -0.1725 -0.5984>, rot=<0.006 0.0168 -0.0329 -0.9993>}
 middle_01_l {parent=middle_metacarpal_l, loc=<7.3275 -0.2447 0.0129>, rot=<-0.0208 0.0042 0.0858 -0.9961>}
 middle_02_l {parent=middle_01_l, loc=<4.0123 0.1401 0.1175>, rot=<-0.0218 0.0258 0.3512 -0.9357>}
 middle_03_l {parent=middle_02_l, loc=<2.6077 0.0534 0.1373>, rot=<0.0082 0.0188 0.1708 -0.9851>}
 pinky_metacarpal_l {parent=hand_l, loc=<1.7616 -0.9753 3.3814>, rot=<-0.0397 0.0814 -0.069 -0.9935>}
 pinky_01_l {parent=pinky_metacarpal_l, loc=<7.1709 -0.4831 -0.1094>, rot=<-0.0654 -0.0035 0.2242 -0.9723>}
 pinky_02_l {parent=pinky_01_l, loc=<2.8274 0.0893 -0.1043>, rot=<0.0037 0.0011 0.2854 -0.9584>}
 pinky_03_l {parent=pinky_02_l, loc=<2.2046 0.0307 -0.0078>, rot=<-0.0002 0.0098 0.1445 -0.9895>}
 ring_metacarpal_l {parent=hand_l, loc=<2.2074 -0.3625 1.3638>, rot=<-0.0284 0.0722 -0.0458 -0.9959>}
 ring_01_l {parent=ring_metacarpal_l, loc=<7.0295 -0.6197 -0.1446>, rot=<-0.0489 -0.0252 0.1892 -0.9804>}
 ring_02_l {parent=ring_01_l, loc=<3.6174 0.4647 -0.1027>, rot=<0 0 0.2806 -0.9598>}
 ring_03_l {parent=ring_02_l, loc=<2.4152 0.114 0.0199>, rot=<0 0 0.1189 -0.9929>}
 thumb_01_l {parent=hand_l, loc=<1.3118 -0.3191 -2.459>, rot=<0.6546 -0.4042 0.0664 -0.6354>}
 thumb_02_l {parent=thumb_01_l, loc=<5.9336 -0.0448 0.0401>, rot=<-0.0037 0.0213 0.298 -0.9543>}
 thumb_03_l {parent=thumb_02_l, loc=<2.2223 0.3682 -0.0907>, rot=<0.0118 -0.013 0.0066 -0.9998>}
 lowerarm_twist_01_l {parent=lowerarm_l, loc=<18.316 0 0>, rot=<0 0 0 1>}
 lowerarm_twist_02_l {parent=lowerarm_l, loc=<10.6707 0 0>, rot=<0 0 0 1>}
 upperarm_twist_01_l {parent=upperarm_l, loc=<7.2135 -0.1134 0>, rot=<0 0 0 1>}
 upperarm_twist_02_l {parent=upperarm_l, loc=<19.7998 -0.1134 0>, rot=<0 0 0 1>}
 shoulder_pad_l_MSTR {parent=clavicle_l, loc=<9.8676 -1.2426 -0.4101>, rot=<-0.01 0.468 -0.0318 -0.8831>}
 shoulder_pad_l {parent=shoulder_pad_l_MSTR, loc=<0.0225 0 0>, rot=<0 0 0 1>}
 shoulder_pad_front_l {parent=shoulder_pad_l, loc=<-2.338 7.2415 -0.6827>, rot=<0 0 -0.5871 -0.8095>}
 shoulder_pad_back_l {parent=shoulder_pad_l, loc=<-3.8978 -6.1118 0>, rot=<0 0 0.8379 -0.5458>}
 shoulder_pad_top_l {parent=shoulder_pad_l, loc=<4.6692 -1.3745 0>, rot=<0 0 0.0537 -0.9986>}
 clavicle_r {parent=spine_05, loc=<5.4831 -0.1053 3.2946>, rot=<0.7289 0.0465 -0.6829 0.009>}
 upperarm_r {parent=clavicle_r, loc=<-9.8658 1.2136 0.4104>, rot=<0.0164 -0.3245 -0.0269 -0.9454>}
 lowerarm_r {parent=upperarm_r, loc=<-26.4651 0.8345 -0.259>, rot=<0.0036 0.0161 -0.2298 -0.9731>}
 hand_r {parent=lowerarm_r, loc=<-22.7356 0.3203 0.6106>, rot=<-0.6784 -0.0223 0.0735 -0.7307>}
 index_metacarpal_r {parent=hand_r, loc=<-1.6534 -0.0181 2.9601>, rot=<0.0158 -0.0124 -0.0173 -0.9996>}
 index_01_r {parent=index_metacarpal_r, loc=<-8.3008 -0.0357 0.0524>, rot=<-0.0044 0.0485 0.0924 -0.9945>}
 index_02_r {parent=index_01_r, loc=<-3.3589 0.3003 -0.0453>, rot=<-0.002 0.0027 0.3225 -0.9466>}
 index_03_r {parent=index_02_r, loc=<-2.5578 0.123 0.0252>, rot=<-0.002 0.0067 0.2558 -0.9667>}
 middle_metacarpal_r {parent=hand_r, loc=<-2.1856 0.1726 0.5984>, rot=<0.006 0.0168 -0.0329 -0.9993>}
 middle_01_r {parent=middle_metacarpal_r, loc=<-7.3273 0.2447 -0.0129>, rot=<-0.0208 0.0042 0.0858 -0.9961>}
 middle_02_r {parent=middle_01_r, loc=<-4.0124 -0.14 -0.1175>, rot=<-0.0218 0.0258 0.3512 -0.9357>}
 middle_03_r {parent=middle_02_r, loc=<-2.6077 -0.0533 -0.1373>, rot=<0.0082 0.0188 0.1708 -0.9851>}
 pinky_metacarpal_r {parent=hand_r, loc=<-1.7617 0.9753 -3.3814>, rot=<-0.0397 0.0814 -0.069 -0.9935>}
 pinky_01_r {parent=pinky_metacarpal_r, loc=<-7.1707 0.4832 0.1094>, rot=<-0.0654 -0.0035 0.2242 -0.9723>}
 pinky_02_r {parent=pinky_01_r, loc=<-2.8274 -0.0892 0.1043>, rot=<0.0037 0.0011 0.2854 -0.9584>}
 pinky_03_r {parent=pinky_02_r, loc=<-2.2045 -0.031 0.0078>, rot=<-0.0002 0.0098 0.1445 -0.9895>}
 ring_metacarpal_r {parent=hand_r, loc=<-2.2073 0.3627 -1.3638>, rot=<-0.0284 0.0722 -0.0458 -0.9959>}
 ring_01_r {parent=ring_metacarpal_r, loc=<-7.0297 0.6197 0.1446>, rot=<-0.0489 -0.0252 0.1892 -0.9804>}
 ring_02_r {parent=ring_01_r, loc=<-3.6171 -0.4648 0.1027>, rot=<0 0 0.2806 -0.9598>}
 ring_03_r {parent=ring_02_r, loc=<-2.4153 -0.1141 -0.0199>, rot=<0 0 0.1189 -0.9929>}
 thumb_01_r {parent=hand_r, loc=<-1.3119 0.3192 2.459>, rot=<0.6546 -0.4042 0.0664 -0.6354>}
 thumb_02_r {parent=thumb_01_r, loc=<-5.9335 0.0447 -0.0401>, rot=<-0.0037 0.0213 0.298 -0.9543>}
 thumb_03_r {parent=thumb_02_r, loc=<-2.2223 -0.3682 0.0907>, rot=<0.0118 -0.013 0.0066 -0.9998>}
 lowerarm_twist_01_r {parent=lowerarm_r, loc=<-18.3161 0 0>, rot=<0 0 0 1>}
 lowerarm_twist_02_r {parent=lowerarm_r, loc=<-10.6707 0 0>, rot=<0 0 0 1>}
 upperarm_twist_01_r {parent=upperarm_r, loc=<-7.2135 0.1134 0>, rot=<0 0 0 1>}
 upperarm_twist_02_r {parent=upperarm_r, loc=<-19.7998 0.1134 0>, rot=<0 0 0 1>}
 shoulder_pad_r_MSTR {parent=clavicle_r, loc=<-9.8672 1.2426 0.4102>, rot=<0.0176 -0.9164 -0.0022 -0.3999>}
 shoulder_pad_r {parent=shoulder_pad_r_MSTR, loc=<0.0225 0 0>, rot=<0 0 0 1>}
 shoulder_pad_back_r {parent=shoulder_pad_r, loc=<-3.7112 6.1842 -0.72>, rot=<0.07 0.0316 -0.8296 -0.553>}
 shoulder_pad_front_r {parent=shoulder_pad_r, loc=<-2.3744 -7.2098 -0.8689>, rot=<-0.0351 0.0683 0.5939 -0.8008>}
 shoulder_pad_top_r {parent=shoulder_pad_r, loc=<4.6462 1.2824 0.6809>, rot=<0.0157 0.0751 -0.0429 -0.9961>}
 neck_01 {parent=spine_05, loc=<11.3138 0.1529 0>, rot=<0 0 -0.2003 -0.9797>}
 neck_02 {parent=neck_01, loc=<4.5627 0 0>, rot=<0 0 0.0504 -0.9987>}
 head {parent=neck_02, loc=<5.3914 0 0>, rot=<0 0 0.115 -0.9934>}
 dyn_neck_heart_2_mstr {parent=neck_02, loc=<-2.2655 4.4825 0>, rot=<0.5224 0.4766 -0.4766 -0.5224>}
 dyn_neck_heart_2 {parent=dyn_neck_heart_2_mstr, loc=<0 0 0>, rot=<0 0 0 1>}
 dyn_neck_heart_1_mstr {parent=neck_01, loc=<-0.0336 5.4296 -0.0465>, rot=<-0.6175 -0.3445 -0.3445 -0.6175>}
 dyn_neck_heart_1 {parent=dyn_neck_heart_1_mstr, loc=<0 0 0>, rot=<0 0 0 1>}
 collar_l_front_MSTR {parent=spine_05, loc=<6.1194 5.7646 -4.8077>, rot=<0.0335 -0.1854 -0.2093 -0.9595>}
 collar_l_front_01 {parent=collar_l_front_MSTR, loc=<0.15 0 0>, rot=<0 0 0 1>}
 collar_l_front_02 {parent=collar_l_front_01, loc=<5.1776 0 0>, rot=<0 -0.2538 0 -0.9673>}
 collar_l_front_03 {parent=collar_l_front_02, loc=<4.9199 0 0>, rot=<0 0 0 1>}
 collar_l_MSTR {parent=spine_05, loc=<10.6765 0.9825 -5.3608>, rot=<0.0276 -0.1589 -0.2105 -0.9642>}
 collar_l_01 {parent=collar_l_MSTR, loc=<0.15 0 0>, rot=<0 -0.0427 0 -0.9991>}
 collar_l_02 {parent=collar_l_01, loc=<4.389 0 0>, rot=<0 -0.2538 0 -0.9673>}
 collar_l_03 {parent=collar_l_02, loc=<4.9199 0 0>, rot=<0 0 0 1>}
 collar_l_04 {parent=collar_l_02, loc=<4.5201 -6.0669 1.4199>, rot=<0.1572 -0.0244 0.1516 -0.9756>}
 collar_l_back_MSTR {parent=spine_05, loc=<12.3448 -4.118 0>, rot=<0.7027 0.0788 -0.0788 -0.7027>}
 collar_l_back_01 {parent=collar_l_back_MSTR, loc=<0.15 0 0>, rot=<0 -0.0427 0 -0.9991>}
 collar_l_back_02 {parent=collar_l_back_01, loc=<3.9754 0 0>, rot=<0 -0.2487 0 -0.9686>}
 collar_l_back_03 {parent=collar_l_back_02, loc=<6.9308 0 -0.038>, rot=<0 -0.0645 0 -0.9979>}
 collar_l_back_l {parent=collar_l_back_02, loc=<6.0531 5.6312 0.8668>, rot=<-0.1216 -0.0884 -0.1434 -0.9782>}
 collar_l_back_r {parent=collar_l_back_02, loc=<6.0531 -5.631 0.8668>, rot=<0.1522 -0.0803 0.1471 -0.974>}
 collar_r_front_MSTR {parent=spine_05, loc=<6.1194 5.7646 4.808>, rot=<0.9599 0.2077 0.1826 -0.0464>}
 collar_r_front_01 {parent=collar_r_front_MSTR, loc=<0.15 0 0>, rot=<0 0 0 1>}
 collar_r_front_02 {parent=collar_r_front_01, loc=<5.1776 0 0>, rot=<0 -0.2538 0 -0.9673>}
 collar_r_front_03 {parent=collar_r_front_02, loc=<4.9199 0 0>, rot=<0 0 0 1>}
 collar_r_MSTR {parent=spine_05, loc=<10.6765 0.9825 5.361>, rot=<0.9637 0.2128 0.1558 -0.0414>}
 collar_r_01 {parent=collar_r_MSTR, loc=<0.15 0 0>, rot=<0 -0.0427 0 -0.9991>}
 collar_r_02 {parent=collar_r_01, loc=<4.389 0 0>, rot=<0 -0.2538 0 -0.9673>}
 collar_r_03 {parent=collar_r_02, loc=<4.9199 0 0>, rot=<0 0 0 1>}
 collar_r_04 {parent=collar_r_02, loc=<4.4828 5.8992 1.509>, rot=<-0.1469 -0.0579 -0.1623 -0.974>}
 dyn_backpack {parent=spine_05, loc=<0.5113 -6.3593 0>, rot=<0.5196 -0.4795 0.4795 -0.5196>}
 attach_backpack {parent=spine_05, loc=<-0.123 -13.473 0>, rot=<-0.0093 -0.707 0.0093 -0.707>}
 attach_cape {parent=spine_05, loc=<9.017 -11.714 0>, rot=<0.0616 -0.7044 -0.0616 -0.7044>}
 thigh_l {parent=pelvis, loc=<-1.0855 -0.2109 -8.7053>, rot=<0.0009 0.0355 0.0155 -0.9993>}
 calf_l {parent=thigh_l, loc=<-42.5284 -0.0019 -0.7297>, rot=<0.0006 -0.0161 -0.0391 -0.9991>}
 calf_twist_01_l {parent=calf_l, loc=<-35.8483 -1.9197 -0.5794>, rot=<0.0237 0.0095 -0.0035 -0.9997>}
 tassel_l_09 {parent=calf_twist_01_l, loc=<2.8995 2.3127 -2.7627>, rot=<-0.6586 0.2662 0.6728 -0.2067>}
 calf_twist_02_l {parent=calf_l, loc=<-19.1546 -1.7943 -0.2638>, rot=<0.0237 0.0095 -0.0035 -0.9997>}
 tassel_l_08 {parent=calf_twist_02_l, loc=<-7.0761 0.866 -3.8105>, rot=<-0.6477 0.2916 0.6643 -0.2326>}
 tassel_l_07 {parent=calf_twist_02_l, loc=<-0.3115 0.1196 -5.0015>, rot=<-0.6545 0.2761 0.6697 -0.2167>}
 foot_l {parent=calf_l, loc=<-46.7402 -2.5014 -0.592>, rot=<0.0427 -0.0238 0.0353 -0.9982>}
 ball_l {parent=foot_l, loc=<-3.5658 10.8066 -0.1544>, rot=<0.0001 -0.0001 -0.7063 -0.7079>}
 tassel_l_10 {parent=foot_l, loc=<7.4202 6.6767 -1.0854>, rot=<0.7077 -0.2491 -0.6576 -0.0687>}
 tassel_l_11 {parent=foot_l, loc=<1.2636 9.834 -0.8655>, rot=<0.7272 -0.2341 -0.6401 -0.0808>}
 tassel_l_06 {parent=calf_l, loc=<-12.3203 -1.8938 -5.3171>, rot=<-0.6162 0.3692 0.6229 -0.3098>}
 tassel_l_05 {parent=calf_l, loc=<-5.4046 -1.7274 -4.1535>, rot=<-0.5751 0.4304 0.5878 -0.3721>}
 thigh_twist_01_l {parent=thigh_l, loc=<-27.952 0 0>, rot=<0.0013 0 0 -1>}
 tassel_l_01 {parent=thigh_twist_01_l, loc=<7.221 -0.4935 -6.8783>, rot=<0.5482 -0.4467 -0.5482 0.4467>}
 tassel_l_02 {parent=thigh_twist_01_l, loc=<0.3903 -0.3006 -5.9265>, rot=<0.5482 -0.4467 -0.5482 0.4467>}
 tassel_l_03 {parent=thigh_l, loc=<-34.3429 -0.7412 -4.938>, rot=<0.5476 -0.4474 -0.5476 0.4474>}
 tassel_l_04 {parent=thigh_l, loc=<-41.0478 -1.2055 -4.4904>, rot=<-0.5666 0.4231 0.5666 -0.4231>}
 thigh_r {parent=pelvis, loc=<-1.0853 -0.2109 8.7053>, rot=<0.0355 -0.0009 0.9993 0.0155>}
 calf_r {parent=thigh_r, loc=<42.5286 0.0019 0.7297>, rot=<0.0006 -0.0161 -0.0391 -0.9991>}
 calf_twist_01_r {parent=calf_r, loc=<35.8485 1.9197 0.5794>, rot=<0.0237 0.0095 -0.0035 -0.9997>}
 tassel_r_09 {parent=calf_twist_01_r, loc=<-2.8997 -2.3127 2.7627>, rot=<-0.6586 0.2662 0.6728 -0.2067>}
 calf_twist_02_r {parent=calf_r, loc=<19.1546 1.7943 0.2638>, rot=<0.0237 0.0095 -0.0035 -0.9997>}
 tassel_r_07 {parent=calf_twist_02_r, loc=<0.3115 -0.1196 5.0015>, rot=<-0.6545 0.2761 0.6697 -0.2167>}
 tassel_r_08 {parent=calf_twist_02_r, loc=<7.076 -0.866 3.8105>, rot=<-0.6477 0.2916 0.6643 -0.2326>}
 foot_r {parent=calf_r, loc=<46.7403 2.5014 0.592>, rot=<0.0427 -0.0238 0.0353 -0.9982>}
 ball_r {parent=foot_r, loc=<3.5658 -10.8066 0.1544>, rot=<0.0001 -0.0001 -0.7063 -0.7079>}
 tassel_r_10 {parent=foot_r, loc=<-7.4203 -6.6767 1.0855>, rot=<0.7077 -0.2491 -0.6576 -0.0687>}
 tassel_r_11 {parent=foot_r, loc=<-1.2638 -9.834 0.8655>, rot=<0.7272 -0.2341 -0.6401 -0.0808>}
 tassel_r_05 {parent=calf_r, loc=<5.4045 1.7274 4.1535>, rot=<-0.5751 0.4304 0.5878 -0.3721>}
 tassel_r_06 {parent=calf_r, loc=<12.3203 1.8938 5.3171>, rot=<-0.6162 0.3692 0.6229 -0.3098>}
 thigh_twist_01_r {parent=thigh_r, loc=<27.952 0 0>, rot=<0.0013 0 0 -1>}
 tassel_r_02 {parent=thigh_twist_01_r, loc=<-0.3902 0.3006 5.9265>, rot=<-0.5482 0.4467 0.5482 -0.4467>}
 tassel_r_01 {parent=thigh_twist_01_r, loc=<-7.2209 0.4935 6.8782>, rot=<-0.5482 0.4467 0.5482 -0.4467>}
 tassel_r_04 {parent=thigh_r, loc=<41.048 1.2055 4.4904>, rot=<-0.5666 0.4231 0.5666 -0.4231>}
 tassel_r_03 {parent=thigh_r, loc=<34.3431 0.7412 4.938>, rot=<-0.5476 0.4474 0.5476 -0.4474>}
 skirt_l_MSTR {parent=pelvis, loc=<-1.4391 -11.6101 -8.6973>, rot=<-0.0155 -0.9993 0.0009 -0.0355>}
 skirt_l_01 {parent=skirt_l_MSTR, loc=<0.2542 0 0>, rot=<0 0 0 1>}
 skirt_l_02 {parent=skirt_l_01, loc=<24.3831 0 0>, rot=<0 0 0 1>}
 skirt_r_MSTR {parent=pelvis, loc=<-1.4391 -11.6101 8.697>, rot=<-0.0155 -0.9993 -0.0009 0.0355>}
 skirt_r_01 {parent=skirt_r_MSTR, loc=<0.254 0 0>, rot=<0 0 0 1>}
 skirt_r_02 {parent=skirt_r_01, loc=<24.3831 0 0>, rot=<0 0 0 1>}
 skirt_MSTR {parent=pelvis, loc=<-1.4678 -12.5346 0>, rot=<0.1004 -0.9949 0 0>}
 skirt_01 {parent=skirt_MSTR, loc=<0.254 0 0>, rot=<0 0 0 1>}
 skirt_02 {parent=skirt_01, loc=<16.5998 0 0>, rot=<0 0 -0.074 -0.9973>}
 skirt_03 {parent=skirt_02, loc=<10.8639 0 0>, rot=<0 0 -0.0462 -0.9989>}
 skirt_04 {parent=skirt_03, loc=<8.8223 0 0>, rot=<0 0 0 1>}
]
variables [Skydive Float {type=double, default=0}]
