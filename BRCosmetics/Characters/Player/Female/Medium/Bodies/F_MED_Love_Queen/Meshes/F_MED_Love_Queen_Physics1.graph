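# Physics asset for the F_MED_Love_Queen female body cosmetic, covering its collar pieces and shoulder pads, apparently for secondary motion.
# Collar chains run from small MSTR anchor bodies through segments with tight swing and twist limits, several branching at the 02 segment.
# Shoulder pads connect to their anchors with free rotation and carry angle-limited top pieces.
# All anchors and a large head capsule constrain to spine_05 with free rotation.
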PhysicsAsset /BRCosmetics/Characters/Player/Female/Medium/Bodies/F_MED_Love_Queen/Meshes/F_MED_Love_Queen_Physics1
bodies {
 collar_l_01 [{shapeType=Capsule, length=1.5, radius=1, center=<0.01 0 0>, rotation=<rot 90 0 0>}]
 collar_l_02 [{shapeType=Capsule, length=0.6253, radius=2.1354, center=<0.1 0 0>, rotation=<rot 90 0 0>}]
 collar_l_03 [{shapeType=Capsule, length=2.3417, radius=1.7998, center=<1 0 0>, rotation=<rot 90 0 0>}]
 collar_l_04 [{shapeType=Capsule, length=2.3417, radius=1.7998, center=<1 0 0>, rotation=<rot 90 0 0>}]
 collar_l_MSTR [{shapeType=Capsule, length=0.1, radius=0.1, center=<0.01 0 0>, rotation=<rot 90 0 0>}]
 collar_l_back_01 [{shapeType=Capsule, length=1.5, radius=1, center=<0.01 0 0>, rotation=<rot 90 0 0>}]
 collar_l_back_02 [{shapeType=Capsule, length=0.0003, radius=2.7604, center=<0.1 0 0>, rotation=<rot 90 0 0>}]
 collar_l_back_03 [{shapeType=Capsule, length=0.4224, radius=1, center=<0.3082 0 0>, rotation=<rot 90 0 0>}]
 collar_l_back_MSTR [{shapeType=Capsule, length=0.1, radius=0.1, center=<0.01 0 0>, rotation=<rot 90 0 0>}]
 collar_l_back_l [{shapeType=Capsule, length=0.4224, radius=1, center=<0.2778 0 0>, rotation=<rot 90 0 0>}]
 collar_l_back_r [{shapeType=Capsule, length=2.5, radius=1.75, center=<0.3705 -1.5 0>, rotation=<rot 90 0 0>}]
 collar_l_front_01 [{shapeType=Capsule, length=1.5, radius=1, center=<0.01 0 0>, rotation=<rot 90 0 0>}]
 collar_l_front_02 [{shapeType=Capsule, length=0, radius=2.8033, center=<1 0 0>, rotation=<rot 90 0 0>}]
 collar_l_front_03 [{shapeType=Capsule, length=2.3417, radius=1.7998, center=<1 0 0>, rotation=<rot 90 0 0>}]
 collar_l_front_MSTR [{shapeType=Capsule, length=0.1, radius=0.1, center=<0.01 0 0>, rotation=<rot 90 0 0>}]
 collar_r_01 [{shapeType=Capsule, length=1.5, radius=1, center=<0.01 0 0>, rotation=<rot 90 0 0>}]
 collar_r_02 [{shapeType=Capsule, length=0.0003, radius=2.7604, center=<0.3 0 0>, rotation=<rot 90 0 0>}]
 collar_r_03 [{shapeType=Capsule, length=2.3417, radius=1.7998, center=<1 0.5 0>, rotation=<rot 90 0 0>}]
 collar_r_04 [{shapeType=Capsule, length=2.3417, radius=1.7998, center=<1 -1 0>, rotation=<rot 90 0 0>}]
 collar_r_MSTR [{shapeType=Capsule, length=0.1, radius=0.1, center=<0.01 0 0>, rotation=<rot 90 0 0>}]
 collar_r_front_01 [{shapeType=Capsule, length=1.5, radius=1, center=<0.01 0 0>, rotation=<rot 90 0 0>}]
 collar_r_front_02 [{shapeType=Capsule, length=0.0003, radius=2.7604, center=<0.3 0 0>, rotation=<rot 90 0 0>}]
 collar_r_front_03 [{shapeType=Capsule, length=2.8561, radius=1.5537, center=<0.1 0 0>, rotation=<rot 90 0 0>}]
 collar_r_front_MSTR [{shapeType=Capsule, length=0.1, radius=0.1, center=<0.01 0 0>, rotation=<rot 90 0 0>}]
 head [{shapeType=Capsule, length=2.5, radius=8.375, shapeName=head_capsule, center=<3.7915 0 -0.0001>, rotation=<rot 90 0 0>}]
 shoulder_pad_l [{shapeType=Capsule, length=1.369, radius=3.6985, center=<0.1 0 0>, rotation=<rot 0 0 90>}]
 shoulder_pad_l_MSTR [{shapeType=Capsule, length=0.1, radius=0.1, rotation=<rot 0 -90 0>}]
 shoulder_pad_r [{shapeType=Capsule, length=1.369, radius=3.4587, center=<0.1 0 0>, rotation=<rot 0 0 90>}]
 shoulder_pad_r_MSTR [{shapeType=Capsule, length=0.1, radius=0.1, rotation=<rot 0 -90 0>}]
 shoulder_pad_top_l [
  {shapeType=Capsule, length=5.6227, radius=6.1084, rotation=<rot 0 5.0019 90>}
  {shapeType=Capsule, length=6.5174, radius=4.8593, center=<-4.3802 5.1592 -0.7348>, rotation=<rot 80.6694 157.8998 179.9998>}
 ]
 shoulder_pad_top_r [{shapeType=Capsule, length=6.3833, radius=5.4124, rotation=<rot 0 0 90>}]
 spine_05 [{shapeType=Capsule, length=0.1, radius=0.1}]
}
constraints [
 {
  bone1Name=shoulder_pad_r
  bone2Name=shoulder_pad_r_MSTR
  swing1LimitDegrees=45
  swing1Motion=Free
  swing2LimitDegrees=45
  swing2Motion=Free
  twistLimitDegrees=45
  twistMotion=Free
 }
 {
  bone1Name=shoulder_pad_l
  bone2Name=shoulder_pad_l_MSTR
  swing1LimitDegrees=45
  swing1Motion=Free
  swing2LimitDegrees=45
  swing2Motion=Free
  twistLimitDegrees=45
  twistMotion=Free
 }
 {bone1Name=shoulder_pad_top_l, bone2Name=shoulder_pad_l, swing1LimitDegrees=25, swing2LimitDegrees=25, twistLimitDegrees=3}
 {bone1Name=shoulder_pad_top_r, bone2Name=shoulder_pad_r, swing1LimitDegrees=25, swing2LimitDegrees=25, twistLimitDegrees=3}
 {bone1Name=collar_l_front_01, bone2Name=collar_l_front_MSTR, swing1LimitDegrees=25, swing2LimitDegrees=25, twistLimitDegrees=3}
 {bone1Name=collar_l_front_02, bone2Name=collar_l_front_01, swing1LimitDegrees=10, swing2LimitDegrees=15, twistLimitDegrees=3}
 {bone1Name=collar_l_front_03, bone2Name=collar_l_front_02, swing1LimitDegrees=15, swing2LimitDegrees=25, twistLimitDegrees=3}
 {bone1Name=collar_l_01, bone2Name=collar_l_MSTR, swing1LimitDegrees=25, swing2LimitDegrees=25, twistLimitDegrees=3}
 {bone1Name=collar_l_02, bone2Name=collar_l_01, swing1LimitDegrees=10, swing2LimitDegrees=15, twistLimitDegrees=3}
 {bone1Name=collar_l_03, bone2Name=collar_l_02, swing1LimitDegrees=25, swing2LimitDegrees=25, twistLimitDegrees=3}
 {bone1Name=collar_l_04, bone2Name=collar_l_02, swing1LimitDegrees=25, swing2LimitDegrees=25, twistLimitDegrees=3}
 {bone1Name=collar_l_back_01, bone2Name=collar_l_back_MSTR, swing1LimitDegrees=25, swing2LimitDegrees=25, twistLimitDegrees=3}
 {bone1Name=collar_l_back_02, bone2Name=collar_l_back_01, swing1LimitDegrees=10, swing2LimitDegrees=15, twistLimitDegrees=1}
 {bone1Name=collar_l_back_03, bone2Name=collar_l_back_02, swing1LimitDegrees=15, swing2LimitDegrees=25, twistLimitDegrees=3}
 {bone1Name=collar_l_back_l, bone2Name=collar_l_back_02, swing1LimitDegrees=15, swing2LimitDegrees=25, twistLimitDegrees=3}
 {bone1Name=collar_l_back_r, bone2Name=collar_l_back_02, swing1LimitDegrees=15, swing2LimitDegrees=25, twistLimitDegrees=3}
 {bone1Name=collar_r_front_01, bone2Name=collar_r_front_MSTR, swing1LimitDegrees=25, swing2LimitDegrees=25, twistLimitDegrees=3}
 {bone1Name=collar_r_front_02, bone2Name=collar_r_front_01, swing1LimitDegrees=10, swing2LimitDegrees=15, twistLimitDegrees=3}
 {bone1Name=collar_r_front_03, bone2Name=collar_r_front_02, swing1LimitDegrees=15, swing2LimitDegrees=25, twistLimitDegrees=3}
 {bone1Name=collar_r_01, bone2Name=collar_r_MSTR, swing1LimitDegrees=25, swing2LimitDegrees=25, twistLimitDegrees=3}
 {bone1Name=collar_r_02, bone2Name=collar_r_01, swing1LimitDegrees=10, swing2LimitDegrees=15, twistLimitDegrees=15}
 {bone1Name=collar_r_03, bone2Name=collar_r_02, swing1LimitDegrees=15, swing2LimitDegrees=25, twistLimitDegrees=3}
 {bone1Name=collar_r_04, bone2Name=collar_r_02, swing1LimitDegrees=25, swing2LimitDegrees=25, twistLimitDegrees=3}
 {
  bone1Name=shoulder_pad_l_MSTR
  bone2Name=spine_05
  swing1LimitDegrees=45
  swing1Motion=Free
  swing2LimitDegrees=45
  swing2Motion=Free
  twistLimitDegrees=45
  twistMotion=Free
 }
 {
  bone1Name=shoulder_pad_r_MSTR
  bone2Name=spine_05
  swing1LimitDegrees=45
  swing1Motion=Free
  swing2LimitDegrees=45
  swing2Motion=Free
  twistLimitDegrees=45
  twistMotion=Free
 }
 {
  bone1Name=collar_l_front_MSTR
  bone2Name=spine_05
  swing1LimitDegrees=45
  swing1Motion=Free
  swing2LimitDegrees=45
  swing2Motion=Free
  twistLimitDegrees=45
  twistMotion=Free
 }
 {
  bone1Name=collar_l_MSTR
  bone2Name=spine_05
  swing1LimitDegrees=45
  swing1Motion=Free
  swing2LimitDegrees=45
  swing2Motion=Free
  twistLimitDegrees=45
  twistMotion=Free
 }
 {
  bone1Name=collar_l_back_MSTR
  bone2Name=spine_05
  swing1LimitDegrees=45
  swing1Motion=Free
  swing2LimitDegrees=45
  swing2Motion=Free
  twistLimitDegrees=45
  twistMotion=Free
 }
 {
  bone1Name=collar_r_front_MSTR
  bone2Name=spine_05
  swing1LimitDegrees=45
  swing1Motion=Free
  swing2LimitDegrees=45
  swing2Motion=Free
  twistLimitDegrees=45
  twistMotion=Free
 }
 {
  bone1Name=collar_r_MSTR
  bone2Name=spine_05
  swing1LimitDegrees=45
  swing1Motion=Free
  swing2LimitDegrees=45
  swing2Motion=Free
  twistLimitDegrees=45
  twistMotion=Free
 }
 {
  bone1Name=head
  bone2Name=spine_05
  swing1LimitDegrees=45
  swing1Motion=Free
  swing2LimitDegrees=45
  swing2Motion=Free
  twistLimitDegrees=45
  twistMotion=Free
 }
]
