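# Physics asset for the F_MED_Love_Queen skirt, providing capsule bodies and constraints for the skirt and leg tassels.
# A center four-bone skirt chain and two-bone side chains hang from MSTR anchors on the pelvis.
# Eleven tassels per leg each constrain individually to thigh, calf, twist or foot bodies, with narrow twist limits like the skirt.
# Pelvis and leg capsules give body volume, while tiny MSTR and twist capsules act as angularly free attachment points.
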PhysicsAsset /BRCosmetics/Characters/Player/Female/Medium/Bodies/F_MED_Love_Queen/Meshes/F_MED_Love_Queen_Skirt_Physics
bodies {
 calf_l [{shapeType=Capsule, length=24.3791, radius=7.8249, center=<-21.8174 -2.6715 -0.9122>, rotation=<rot 90 0 0>}]
 calf_r [{shapeType=Capsule, length=24.3787, radius=7.8252, center=<21.8119 2.7048 0.9148>, rotation=<rot 90 0 0>}]
 calf_twist_01_l [{shapeType=Capsule, length=0.1, radius=0.1}]
 calf_twist_01_r [{shapeType=Capsule, length=0.1, radius=0.1}]
 calf_twist_02_l [{shapeType=Capsule, length=0.1, radius=0.1}]
 calf_twist_02_r [{shapeType=Capsule, length=0.1, radius=0.1}]
 foot_l [
  {shapeType=Capsule, length=0.1, radius=0.1, rotation=<rot 0 0 -180>}
  {
   shapeType=Capsule
   length=14.4963
   radius=8.3407
   shapeName=foot_l_capsule
   center=<-1.5202 5.708 0.2454>
   rotation=<rot 0.6328 14.8874 87.6212>
  }
 ]
 foot_r [
  {shapeType=Capsule, length=0.1, radius=0.1}
  {
   shapeType=Capsule
   length=14.4963
   radius=8.3407
   shapeName=foot_r_capsule
   center=<1.5202 -5.708 -0.2454>
   rotation=<rot 0.6328 14.8874 -92.3788>
  }
 ]
 pelvis [{shapeType=Capsule, length=4.8557, radius=8.2239, center=<1.0106 -2.8438 -0.2468>}]
 skirt_01 [{shapeType=Capsule, length=4, radius=4, center=<0.5 0 0>, rotation=<rot 90 0 0>}]
 skirt_02 [{shapeType=Capsule, length=4, radius=4, center=<0.5 0 0>, rotation=<rot 90 0 0>}]
 skirt_03 [{shapeType=Capsule, length=4, radius=4, center=<0.5 0 0>, rotation=<rot 90 0 0>}]
 skirt_04 [{shapeType=Capsule, length=4, radius=4, center=<0.5 0 0>, rotation=<rot 90 0 0>}]
 skirt_MSTR [{shapeType=Capsule, length=0.1, radius=0.1}]
 skirt_l_01 [{shapeType=Capsule, length=4, radius=4, center=<0.5 0 0>, rotation=<rot 90 0 0>}]
 skirt_l_02 [{shapeType=Capsule, length=17.9931, radius=5.1297, center=<0.5 -1.5954 -1.8318>, rotation=<rot 90 0 0>}]
 skirt_l_MSTR [{shapeType=Capsule, length=0.1, radius=0.1}]
 skirt_r_01 [{shapeType=Capsule, length=4, radius=4, center=<0.5 0 0>, rotation=<rot 90 0 0>}]
 skirt_r_02 [{shapeType=Capsule, length=17.9931, radius=5.1297, center=<-0.5 -0.8493 1.8318>, rotation=<rot 90 180 0>}]
 skirt_r_MSTR [{shapeType=Capsule, length=0.1, radius=0.1}]
 tassel_l_01 [{shapeType=Capsule, length=2, radius=1.3, center=<0.03 0 0>, rotation=<rot 90 0 0>}]
 tassel_l_02 [{shapeType=Capsule, length=2, radius=1.3, center=<0.2011 0 0.0282>, rotation=<rot 90 0 0>}]
 tassel_l_03 [{shapeType=Capsule, length=2, radius=1.3, center=<0.1326 0 -0.115>, rotation=<rot 90 0 0>}]
 tassel_l_04 [{shapeType=Capsule, length=2, radius=1.3, center=<0.05 0 0>, rotation=<rot 90 0 0>}]
 tassel_l_05 [{shapeType=Capsule, length=2, radius=1.3, center=<0.3054 0 0>, rotation=<rot 90 0 0>}]
 tassel_l_06 [{shapeType=Capsule, length=2, radius=1.3, center=<0.03 0 0.0015>, rotation=<rot 90 0 0>}]
 tassel_l_07 [{shapeType=Capsule, length=2, radius=1.3, center=<0.03 0 0>, rotation=<rot 90 0 0>}]
 tassel_l_08 [{shapeType=Capsule, length=2, radius=1.3, center=<0.3298 0 -0.0732>, rotation=<rot 90 0 0>}]
 tassel_l_09 [{shapeType=Capsule, length=2, radius=1.3, center=<0.05 0 0>, rotation=<rot 90 0 0>}]
 tassel_l_10 [{shapeType=Capsule, length=2, radius=1.3, center=<0.05 0 0>, rotation=<rot 90 0 0>}]
 tassel_l_11 [{shapeType=Capsule, length=2, radius=1.3, center=<0.0861 0 0>, rotation=<rot 90 0 0>}]
 tassel_r_01 [{shapeType=Capsule, length=2, radius=1.3, center=<-0.03 0 0>, rotation=<rot 90 0 0>}]
 tassel_r_02 [{shapeType=Capsule, length=2, radius=1.3, center=<-0.3377 0.2229 0.0916>, rotation=<rot 90 0 0>}]
 tassel_r_03 [{shapeType=Capsule, length=2, radius=1.3, center=<-0.03 0 0>, rotation=<rot 90 0 0>}]
 tassel_r_04 [{shapeType=Capsule, length=2, radius=1.3, center=<-0.2297 0 0>, rotation=<rot 90 0 0>}]
 tassel_r_05 [{shapeType=Capsule, length=2, radius=1.3, center=<-0.05 0 0>, rotation=<rot 90 0 0>}]
 tassel_r_06 [{shapeType=Capsule, length=2, radius=1.3, center=<-0.05 0 0>, rotation=<rot 90 0 0>}]
 tassel_r_07 [{shapeType=Capsule, length=2, radius=1.3, center=<-0.03 0 0>, rotation=<rot 90 0 0>}]
 tassel_r_08 [{shapeType=Capsule, length=2, radius=1.3, center=<-0.3756 0 0.0536>, rotation=<rot 90 0 0>}]
 tassel_r_09 [{shapeType=Capsule, length=2, radius=1.3, center=<-0.05 0 0>, rotation=<rot 90 0 0>}]
 tassel_r_10 [{shapeType=Capsule, length=2, radius=1.3, center=<-0.03 0 0>, rotation=<rot 90 0 0>}]
 tassel_r_11 [{shapeType=Capsule, length=2, radius=1.3, center=<-0.03 0 0>, rotation=<rot 90 0 0>}]
 thigh_l [{shapeType=Capsule, length=22.649, radius=7.152, center=<-27.1421 -0.2787 0.5506>, rotation=<rot 90 0 0>}]
 thigh_r [{shapeType=Capsule, length=22.782, radius=7.0719, center=<27.1315 0.2478 -0.5473>, rotation=<rot 90 0 0>}]
 thigh_twist_01_l [{shapeType=Capsule, length=0.1, radius=0.1}]
 thigh_twist_01_r [{shapeType=Capsule, length=0.1, radius=0.1}]
}
constraints [
 {bone1Name=thigh_l, bone2Name=pelvis, swing1LimitDegrees=45, swing2LimitDegrees=45, twistLimitDegrees=45}
 {bone1Name=calf_l, bone2Name=thigh_l, swing1LimitDegrees=45, swing2LimitDegrees=45, twistLimitDegrees=45}
 {bone1Name=thigh_r, bone2Name=pelvis, swing1LimitDegrees=45, swing2LimitDegrees=45, twistLimitDegrees=45}
 {bone1Name=calf_r, bone2Name=thigh_r, swing1LimitDegrees=45, swing2LimitDegrees=45, twistLimitDegrees=45}
 {bone1Name=skirt_l_01, bone2Name=skirt_l_MSTR, swing1LimitDegrees=35, swing2LimitDegrees=35, twistLimitDegrees=5}
 {bone1Name=skirt_r_01, bone2Name=skirt_r_MSTR, swing1LimitDegrees=35, swing2LimitDegrees=35, twistLimitDegrees=5}
 {bone1Name=skirt_01, bone2Name=skirt_MSTR, swing1LimitDegrees=35, swing2LimitDegrees=35, twistLimitDegrees=5}
 {bone1Name=skirt_03, bone2Name=skirt_02, swing1LimitDegrees=45, swing2LimitDegrees=45, twistLimitDegrees=5}
 {
  bone1Name=skirt_l_MSTR
  bone2Name=pelvis
  swing1LimitDegrees=45
  swing1Motion=Free
  swing2LimitDegrees=45
  swing2Motion=Free
  twistLimitDegrees=45
  twistMotion=Free
 }
 {bone1Name=skirt_l_02, bone2Name=skirt_l_01, swing1LimitDegrees=45, swing2LimitDegrees=45, twistLimitDegrees=5}
 {
  bone1Name=skirt_r_MSTR
  bone2Name=pelvis
  swing1LimitDegrees=45
  swing1Motion=Free
  swing2LimitDegrees=45
  swing2Motion=Free
  twistLimitDegrees=45
  twistMotion=Free
 }
 {bone1Name=skirt_r_02, bone2Name=skirt_r_01, swing1LimitDegrees=45, swing2LimitDegrees=45, twistLimitDegrees=5}
 {
  bone1Name=skirt_MSTR
  bone2Name=pelvis
  swing1LimitDegrees=45
  swing1Motion=Free
  swing2LimitDegrees=45
  swing2Motion=Free
  twistLimitDegrees=45
  twistMotion=Free
 }
 {bone1Name=skirt_02, bone2Name=skirt_01, swing1LimitDegrees=45, swing2LimitDegrees=45, twistLimitDegrees=5}
 {bone1Name=skirt_04, bone2Name=skirt_03, swing1LimitDegrees=45, swing2LimitDegrees=45, twistLimitDegrees=5}
 {bone1Name=tassel_l_09, bone2Name=calf_twist_01_l, swing1LimitDegrees=35, swing2LimitDegrees=35, twistLimitDegrees=5}
 {bone1Name=tassel_l_08, bone2Name=calf_twist_02_l, swing1LimitDegrees=35, swing2LimitDegrees=35, twistLimitDegrees=5}
 {bone1Name=tassel_l_07, bone2Name=calf_twist_02_l, swing1LimitDegrees=35, swing2LimitDegrees=35, twistLimitDegrees=5}
 {bone1Name=tassel_l_10, bone2Name=foot_l, swing1LimitDegrees=35, swing2LimitDegrees=35, twistLimitDegrees=5}
 {bone1Name=tassel_l_11, bone2Name=foot_l, swing1LimitDegrees=35, swing2LimitDegrees=35, twistLimitDegrees=5}
 {bone1Name=tassel_l_06, bone2Name=calf_l, swing1LimitDegrees=35, swing2LimitDegrees=35, twistLimitDegrees=5}
 {bone1Name=tassel_l_05, bone2Name=calf_l, swing1LimitDegrees=35, swing2LimitDegrees=35, twistLimitDegrees=5}
 {bone1Name=tassel_l_01, bone2Name=thigh_twist_01_l, swing1LimitDegrees=35, swing2LimitDegrees=35, twistLimitDegrees=5}
 {bone1Name=tassel_l_02, bone2Name=thigh_twist_01_l, swing1LimitDegrees=35, swing2LimitDegrees=35, twistLimitDegrees=5}
 {bone1Name=tassel_l_03, bone2Name=thigh_l, swing1LimitDegrees=35, swing2LimitDegrees=35, twistLimitDegrees=5}
 {bone1Name=tassel_l_04, bone2Name=thigh_l, swing1LimitDegrees=35, swing2LimitDegrees=35, twistLimitDegrees=5}
 {bone1Name=tassel_r_09, bone2Name=calf_twist_01_r, swing1LimitDegrees=35, swing2LimitDegrees=35, twistLimitDegrees=5}
 {bone1Name=tassel_r_07, bone2Name=calf_twist_02_r, swing1LimitDegrees=35, swing2LimitDegrees=35, twistLimitDegrees=5}
 {bone1Name=tassel_r_08, bone2Name=calf_twist_02_r, swing1LimitDegrees=35, swing2LimitDegrees=35, twistLimitDegrees=5}
 {bone1Name=tassel_r_10, bone2Name=foot_r, swing1LimitDegrees=35, swing2LimitDegrees=35, twistLimitDegrees=5}
 {bone1Name=tassel_r_11, bone2Name=foot_r, swing1LimitDegrees=35, swing2LimitDegrees=35, twistLimitDegrees=5}
 {bone1Name=tassel_r_05, bone2Name=calf_r, swing1LimitDegrees=35, swing2LimitDegrees=35, twistLimitDegrees=5}
 {bone1Name=tassel_r_06, bone2Name=calf_r, swing1LimitDegrees=35, swing2LimitDegrees=35, twistLimitDegrees=5}
 {bone1Name=tassel_r_02, bone2Name=thigh_twist_01_r, swing1LimitDegrees=35, swing2LimitDegrees=35, twistLimitDegrees=5}
 {bone1Name=tassel_r_01, bone2Name=thigh_twist_01_r, swing1LimitDegrees=35, swing2LimitDegrees=35, twistLimitDegrees=5}
 {bone1Name=tassel_r_04, bone2Name=thigh_r, swing1LimitDegrees=35, swing2LimitDegrees=35, twistLimitDegrees=5}
 {bone1Name=tassel_r_03, bone2Name=thigh_r, swing1LimitDegrees=35, swing2LimitDegrees=35, twistLimitDegrees=5}
 {
  bone1Name=calf_twist_01_l
  bone2Name=calf_l
  swing1LimitDegrees=45
  swing1Motion=Free
  swing2LimitDegrees=45
  swing2Motion=Free
  twistLimitDegrees=45
  twistMotion=Free
 }
 {
  bone1Name=calf_twist_02_l
  bone2Name=calf_l
  swing1LimitDegrees=45
  swing1Motion=Free
  swing2LimitDegrees=45
  swing2Motion=Free
  twistLimitDegrees=45
  twistMotion=Free
 }
 {
  bone1Name=foot_l
  bone2Name=calf_l
  swing1LimitDegrees=45
  swing1Motion=Free
  swing2LimitDegrees=45
  swing2Motion=Free
  twistLimitDegrees=45
  twistMotion=Free
 }
 {
  bone1Name=calf_twist_01_r
  bone2Name=calf_r
  swing1LimitDegrees=45
  swing1Motion=Free
  swing2LimitDegrees=45
  swing2Motion=Free
  twistLimitDegrees=45
  twistMotion=Free
 }
 {
  bone1Name=foot_r
  bone2Name=calf_r
  swing1LimitDegrees=45
  swing1Motion=Free
  swing2LimitDegrees=45
  swing2Motion=Free
  twistLimitDegrees=45
  twistMotion=Free
 }
 {
  bone1Name=calf_twist_02_r
  bone2Name=calf_r
  swing1LimitDegrees=45
  swing1Motion=Free
  swing2LimitDegrees=45
  swing2Motion=Free
  twistLimitDegrees=45
  twistMotion=Free
 }
 {
  bone1Name=thigh_twist_01_r
  bone2Name=thigh_r
  swing1LimitDegrees=45
  swing1Motion=Free
  swing2LimitDegrees=45
  swing2Motion=Free
  twistLimitDegrees=45
  twistMotion=Free
 }
 {
  bone1Name=thigh_twist_01_l
  bone2Name=thigh_l
  swing1LimitDegrees=45
  swing1Motion=Free
  swing2LimitDegrees=45
  swing2Motion=Free
  twistLimitDegrees=45
  twistMotion=Free
 }
]
