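# Animation Blueprint for the F_MED_Love_Queen cosmetic body, correcting collar, neck hearts and skirt atop a pose copied from the parent mesh.
# Two Control Rig passes bracket bone offsets, constraints and two RigidBody simulations.
# EventGraph derives alphas from joint deltas and player state: weapon type scales spine-based physics, skydiving lowers skirt gravity until parachute opens, non-emote frontend pins skirt to thighs.
# Shoulder-pad constraints are unconnected; Rifle ADS correction is unused.
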AnimBlueprint /BRCosmetics/Characters/Player/Female/Medium/Bodies/F_MED_Love_Queen/Meshes/F_MED_Love_Queen_AnimBP
target_skeleton=&/F_MED_Love_Queen_Skeleton
graphs [
 AnimGraph {nodes [
   n1 = "Constraint - shoulder_pad_l_MSTR (RR)"!(ComponentPose, Alpha=1, ConstraintWeights_0=0.5, ConstraintWeights_1:float=0.5) -> (Pose) {
    alphaBoolBlend {blendOption=Linear}
    boneToModify {boneName=shoulder_pad_l_MSTR}
    constraintSetup [{transformType=Rotation, targetBone {boneName=clavicle_l}}, {transformType=Rotation, targetBone {boneName=upperarm_l}}]
    constraintWeights [1,1]
   }
   "Constraint - shoulder_pad_r_MSTR (RR)"!(ComponentPose=@n1, Alpha=1, ConstraintWeights_0=0.5, ConstraintWeights_1:float=0.5) -> (Pose) {
    alphaBoolBlend {blendOption=Linear}
    boneToModify {boneName=shoulder_pad_r_MSTR}
    constraintSetup [{transformType=Rotation, targetBone {boneName=clavicle_r}}, {transformType=Rotation, targetBone {boneName=upperarm_r}}]
    constraintWeights [1,1]
   }
   n2 = Get Skirt FE Alpha() -> (Skirt FE Alpha:double)
   n3 = Control Rig(Source=Copy Pose From Mesh() {bUseAttachedParent=true}) {
    alphaBoolBlend {blendOption=Linear}
    controlRigAssetReference {blueprintRigClass=/BRCosmetics/Characters/Player/Female/Medium/Bodies/F_MED_Love_Queen/Meshes/F_MED_Love_Queen_CtrlRig/F_MED_Love_Queen_CtrlRig.F_MED_Love_Queen_CtrlRig_C}
   }
   n4 = Get backCollarAlpha() -> (backCollarAlpha:double)
   n5 = "Transform (Modify) Bone - Bone: collar_l_back_MSTR"(ComponentPose=Local To Component(LocalPose=@n3), Alpha=@n4, Translation="-2,0,-2", Rotation="0,0,0", Scale="1,1,1") {
    rotationMode=BMM_Additive
    rotationSpace=BCS_BoneSpace
    translationMode=BMM_Additive
    translationSpace=BCS_BoneSpace
    alphaBoolBlend {blendOption=Linear}
    boneToModify {boneName=collar_l_back_MSTR}
    translation {x=-2, z=-2}
   }
   n6 = "Transform (Modify) Bone - Bone: collar_r_MSTR"(ComponentPose=@n5, Alpha=@n4, Translation="-1,1,-1", Rotation="0,0,0", Scale="1,1,1") {
    rotationMode=BMM_Additive
    rotationSpace=BCS_BoneSpace
    translationMode=BMM_Additive
    translationSpace=BCS_BoneSpace
    alphaBoolBlend {blendOption=Linear}
    boneToModify {boneName=collar_r_MSTR}
    translation=<-1 1 -1>
   }
   n7 = "Transform (Modify) Bone - Bone: collar_l_MSTR"(ComponentPose=@n6, Alpha=@n4, Translation="-1,-1,-1", Rotation="0,0,0", Scale="1,1,1") {
    rotationMode=BMM_Additive
    rotationSpace=BCS_BoneSpace
    translationMode=BMM_Additive
    translationSpace=BCS_BoneSpace
    alphaBoolBlend {blendOption=Linear}
    boneToModify {boneName=collar_l_MSTR}
    translation=<-1 -1 -1>
   }
   n8 = "Transform (Modify) Bone - Bone: skirt_r_MSTR"(ComponentPose=@n7, Alpha=Get thigh_r_up() -> (thigh_r_up:double), Translation="0,0,0", Rotation="-30,0,0", Scale="1,1,1") {
    rotationMode=BMM_Additive
    rotationSpace=BCS_ParentBoneSpace
    translationMode=BMM_Additive
    alphaBoolBlend {blendOption=Linear}
    boneToModify {boneName=skirt_r_MSTR}
    rotation {pitch=-30}
   }
   n9 = "Constraint - collar_l_front_MSTR (RR)"!(ComponentPose=@n8, Alpha=1, ConstraintWeights_0=0.7, ConstraintWeights_1:float=0.3) -> (Pose) {
    alphaBoolBlend {blendOption=Linear}
    boneToModify {boneName=collar_l_front_MSTR}
    constraintSetup [{transformType=Rotation, targetBone {boneName=neck_01}}, {transformType=Rotation, targetBone {boneName=neck_02}}]
    constraintWeights [1,1]
   }
   n10 = "Constraint - collar_l_MSTR (RR)"!(ComponentPose=@n9, Alpha=1, ConstraintWeights_0=0.7, ConstraintWeights_1:float=0.3) -> (Pose) {
    alphaBoolBlend {blendOption=Linear}
    boneToModify {boneName=collar_l_MSTR}
    constraintSetup [{transformType=Rotation, targetBone {boneName=neck_01}}, {transformType=Rotation, targetBone {boneName=neck_02}}]
    constraintWeights [1,1]
   }
   n11 = "Constraint - collar_l_back_01 (RR)"!(ComponentPose=@n10, Alpha=1, ConstraintWeights_0=0.7, ConstraintWeights_1:float=0.3) -> (Pose) {
    alphaBoolBlend {blendOption=Linear}
    boneToModify {boneName=collar_l_back_01}
    constraintSetup [{transformType=Rotation, targetBone {boneName=neck_01}}, {transformType=Rotation, targetBone {boneName=neck_02}}]
    constraintWeights [1,1]
   }
   n12 = "Constraint - collar_r_front_01 (RR)"!(ComponentPose=@n11, Alpha=1, ConstraintWeights_0=0.7, ConstraintWeights_1:float=0.3) -> (Pose) {
    alphaBoolBlend {blendOption=Linear}
    boneToModify {boneName=collar_r_front_01}
    constraintSetup [{transformType=Rotation, targetBone {boneName=neck_01}}, {transformType=Rotation, targetBone {boneName=neck_02}}]
    constraintWeights [1,1]
   }
   n13 = "Constraint - collar_r_MSTR (RR)"!(ComponentPose=@n12, Alpha=1, ConstraintWeights_0=0.7, ConstraintWeights_1:float=0.3) -> (Pose) {
    alphaBoolBlend {blendOption=Linear}
    boneToModify {boneName=collar_r_MSTR}
    constraintSetup [{transformType=Rotation, targetBone {boneName=neck_01}}, {transformType=Rotation, targetBone {boneName=neck_02}}]
    constraintWeights [1,1]
   }
   n14 = Get Look Up Collar Correction() -> (Look Up Collar Correction:double)
   n15 = "Transform (Modify) Bone - Bone: collar_l_back_MSTR"(ComponentPose=@n13, Alpha=@n14, Translation="-2,0,-1.5", Rotation="-15,0,0", Scale="1,1,1") {
    rotationMode=BMM_Additive
    rotationSpace=BCS_BoneSpace
    translationMode=BMM_Additive
    translationSpace=BCS_BoneSpace
    alphaBoolBlend {blendOption=Linear}
    boneToModify {boneName=collar_l_back_MSTR}
    rotation {pitch=-15}
    translation {x=-2, z=-1.5}
   }
   n16 = "Transform (Modify) Bone - Bone: collar_r_MSTR"(ComponentPose=@n15, Alpha=@n14, Translation="-1,-1.5,-1", Rotation="0,0,0", Scale="1,1,1") {
    rotationMode=BMM_Additive
    rotationSpace=BCS_BoneSpace
    translationMode=BMM_Additive
    translationSpace=BCS_BoneSpace
    alphaBoolBlend {blendOption=Linear}
    boneToModify {boneName=collar_r_MSTR}
    translation=<-1 -1.5 -1>
   }
   n17 = "Transform (Modify) Bone - Bone: collar_l_MSTR"(ComponentPose=@n16, Alpha=@n14, Translation="-0.5,-0.5,0.5", Rotation="0,0,0", Scale="1,1,1") {
    rotationMode=BMM_Additive
    rotationSpace=BCS_BoneSpace
    translationMode=BMM_Additive
    translationSpace=BCS_BoneSpace
    alphaBoolBlend {blendOption=Linear}
    boneToModify {boneName=collar_l_MSTR}
    translation=<-0.5 -0.5 0.5>
   }
   n18 = "Transform (Modify) Bone - Bone: collar_l_back_02"(ComponentPose=@n17, Alpha=@n14, Translation="-1.5,-1,-1.5", Rotation="-3,0,-5", Scale="1,1,1") {
    rotationMode=BMM_Additive
    rotationSpace=BCS_BoneSpace
    translationMode=BMM_Additive
    translationSpace=BCS_BoneSpace
    alphaBoolBlend {blendOption=Linear}
    boneToModify {boneName=collar_l_back_02}
    rotation {pitch=-3, roll=-5}
    translation=<-1.5 -1 -1.5>
   }
   n19 = "Transform (Modify) Bone - Bone: collar_r_01"(ComponentPose=@n18, Alpha=@n14, Translation="0,0,0", Rotation="-10,5,10", Scale="1,1,1") {
    rotationMode=BMM_Additive
    rotationSpace=BCS_BoneSpace
    translationMode=BMM_Additive
    translationSpace=BCS_BoneSpace
    alphaBoolBlend {blendOption=Linear}
    boneToModify {boneName=collar_r_01}
    rotation=<rot -10 5 10>
   }
   n20 = RigidBody(ComponentPose=@n19, Alpha=Get Pistol Alpha() -> (Pistol Alpha:double)) {
    bClampLinearTranslationLimitToRefPose=true
    bOverrideWorldGravity=true
    overridePhysicsAsset=&/F_MED_Love_Queen_Physics1
    simulationSpace=BaseBoneSpace
    alphaBoolBlend {blendOption=Linear}
    baseBoneRef {boneName=spine_05}
    componentAppliedLinearAccClamp=<3000 3000 3000>
    componentLinearAccScale=<0.1 0.1 0.1>
    componentLinearVelScale=<0.1 0.1 0.1>
    overrideWorldGravity {z=-980}
   }
   n21 = "Constraint - collar_l_front_03 (PP)"!(ComponentPose=@n20, Alpha=0.25, ConstraintWeights_0=1, ConstraintWeights_1:float=0) -> (Pose) {
    alphaBoolBlend {blendOption=Linear}
    boneToModify {boneName=collar_l_front_03}
    constraintSetup [{targetBone {boneName=collar_l_03}}, {targetBone {boneName=collar_l_04}}]
    constraintWeights [1,1]
   }
   n22 = "Constraint - collar_l_03 (PP)"!(ComponentPose=@n21, Alpha=0.5, ConstraintWeights_0=0.5, ConstraintWeights_1:float=0.5) -> (Pose) {
    alphaBoolBlend {blendOption=Linear}
    boneToModify {boneName=collar_l_03}
    constraintSetup [{targetBone {boneName=collar_l_front_03}}, {targetBone {boneName=collar_l_04}}]
    constraintWeights [1,1]
   }
   n23 = "Constraint - collar_l_04 (PP)"!(ComponentPose=@n22, Alpha=0.5, ConstraintWeights_0=0.5, ConstraintWeights_1:float=0.5) -> (Pose) {
    alphaBoolBlend {blendOption=Linear}
    boneToModify {boneName=collar_l_04}
    constraintSetup [{targetBone {boneName=collar_l_03}}, {targetBone {boneName=collar_l_back_l}}]
    constraintWeights [1,1]
   }
   n24 = "Constraint - collar_l_back_l (PP)"!(ComponentPose=@n23, Alpha=0.5, ConstraintWeights_0=0.5, ConstraintWeights_1:float=0.5) -> (Pose) {
    alphaBoolBlend {blendOption=Linear}
    boneToModify {boneName=collar_l_back_l}
    constraintSetup [{targetBone {boneName=collar_l_04}}, {targetBone {boneName=collar_l_back_03}}]
    constraintWeights [1,1]
   }
   n25 = "Constraint - collar_l_back_03 (PP)"!(ComponentPose=@n24, Alpha=0.5, ConstraintWeights_0=0.5, ConstraintWeights_1:float=0.5) -> (Pose) {
    alphaBoolBlend {blendOption=Linear}
    boneToModify {boneName=collar_l_back_03}
    constraintSetup [{targetBone {boneName=collar_l_back_l}}, {targetBone {boneName=collar_l_back_r}}]
    constraintWeights [1,1]
   }
   n26 = "Constraint - collar_l_back_r (PP)"!(ComponentPose=@n25, Alpha=0.5, ConstraintWeights_0=0.5, ConstraintWeights_1:float=0.5) -> (Pose) {
    alphaBoolBlend {blendOption=Linear}
    boneToModify {boneName=collar_l_back_r}
    constraintSetup [{targetBone {boneName=collar_l_back_03}}, {targetBone {boneName=collar_r_04}}]
    constraintWeights [1,1]
   }
   n27 = "Constraint - collar_r_04 (PP)"!(ComponentPose=@n26, Alpha=0.5, ConstraintWeights_0=0.5, ConstraintWeights_1:float=0.5) -> (Pose) {
    alphaBoolBlend {blendOption=Linear}
    boneToModify {boneName=collar_r_04}
    constraintSetup [{targetBone {boneName=collar_l_back_r}}, {targetBone {boneName=collar_r_03}}]
    constraintWeights [1,1]
   }
   n28 = "Constraint - collar_r_03 (PP)"!(ComponentPose=@n27, Alpha=0.5, ConstraintWeights_0=0.5, ConstraintWeights_1:float=0.5) -> (Pose) {
    alphaBoolBlend {blendOption=Linear}
    boneToModify {boneName=collar_r_03}
    constraintSetup [{targetBone {boneName=collar_r_04}}, {targetBone {boneName=collar_r_front_03}}]
    constraintWeights [1,1]
   }
   n29 = "Constraint - collar_r_front_03 (PP)"!(ComponentPose=@n28, Alpha=0.25, ConstraintWeights_0=1, ConstraintWeights_1:float=0) -> (Pose) {
    alphaBoolBlend {blendOption=Linear}
    boneToModify {boneName=collar_r_front_03}
    constraintSetup [{targetBone {boneName=collar_r_03}}, {targetBone {boneName=collar_l_04}}]
    constraintWeights [1,1]
   }
   n30 = RigidBody!(ComponentPose=@n29, Alpha, OverrideWorldGravity:Vector=Get gravityOverride() -> (gravityOverride:Vector), ExternalForce) -> (Pose) {
    bClampLinearTranslationLimitToRefPose=true
    bOverrideWorldGravity=true
    overridePhysicsAsset=&/F_MED_Love_Queen_Skirt_Physics
    simulationSpace=BaseBoneSpace
    alphaBoolBlend {blendOption=Linear}
    baseBoneRef {boneName=root}
    componentAppliedLinearAccClamp=<5000 5000 5000>
    componentLinearAccScale=<0.3 0.2 0.2>
    componentLinearVelScale=<0.2 0.2 0.2>
    overrideWorldGravity {z=-980}
   }
   n31 = "Transform (Modify) Bone - Bone: dyn_backpack"!(ComponentPose=@n30, bAlphaBoolEnabled:bool=Get bIsSkyDiving(), Translation="3,0,0", Rotation="0,0,0", Scale="1,1,1") -> (Pose) {
    alphaInputType=Bool
    translationMode=BMM_Replace
    translationSpace=BCS_BoneSpace
    alphaBoolBlend {blendOption=Linear}
    boneToModify {boneName=dyn_backpack}
    translation {x=3}
   }
   n32 = Get bIsSkyDiving()
   n33 = "Transform (Modify) Bone - Bone: collar_r_front_MSTR"!(ComponentPose=@n31, bAlphaBoolEnabled:bool=@n32, Translation="0,0,0", Rotation="-20,0,0", Scale="1,1,1") -> (Pose) {
    alphaInputType=Bool
    rotationMode=BMM_Replace
    rotationSpace=BCS_BoneSpace
    translationMode=BMM_Replace
    translationSpace=BCS_BoneSpace
    alphaBoolBlend {blendOption=Linear}
    boneToModify {boneName=collar_r_front_MSTR}
    rotation {pitch=-20}
   }
   n34 = "Transform (Modify) Bone - Bone: collar_r_MSTR"!(ComponentPose=@n33, bAlphaBoolEnabled:bool=@n32, Translation="0,0,0", Rotation="-20,0,0", Scale="1,1,1") -> (Pose) {
    alphaInputType=Bool
    rotationMode=BMM_Replace
    rotationSpace=BCS_BoneSpace
    translationMode=BMM_Replace
    translationSpace=BCS_BoneSpace
    alphaBoolBlend {blendOption=Linear}
    boneToModify {boneName=collar_r_MSTR}
    rotation {pitch=-20}
   }
   n35 = "Transform (Modify) Bone - Bone: collar_l_front_MSTR"!(ComponentPose=@n34, bAlphaBoolEnabled:bool=@n32, Translation="0,0,0", Rotation="-100,0,0", Scale="1,1,1") -> (Pose) {
    alphaInputType=Bool
    rotationMode=BMM_Replace
    rotationSpace=BCS_BoneSpace
    translationMode=BMM_Replace
    translationSpace=BCS_BoneSpace
    alphaBoolBlend {blendOption=Linear}
    boneToModify {boneName=collar_l_front_MSTR}
    rotation {pitch=-100}
   }
   n36 = "Transform (Modify) Bone - Bone: collar_l_front_MSTR"(ComponentPose=@n35, bAlphaBoolEnabled=@n32, Translation="0,0,0", Rotation="-100,0,0", Scale="1,1,1") {
    alphaInputType=Bool
    rotationMode=BMM_Replace
    rotationSpace=BCS_BoneSpace
    translationMode=BMM_Replace
    translationSpace=BCS_BoneSpace
    alphaBoolBlend {blendOption=Linear}
    boneToModify {boneName=collar_l_front_MSTR}
    rotation {pitch=-100}
   }
   n37 = "Transform (Modify) Bone - Bone: collar_l_back_MSTR"!(ComponentPose=@n36, bAlphaBoolEnabled:bool=@n32, Translation="0,0,0", Rotation="-15,0,0", Scale="1,1,1") -> (Pose) {
    alphaInputType=Bool
    rotationMode=BMM_Replace
    rotationSpace=BCS_BoneSpace
    translationMode=BMM_Replace
    translationSpace=BCS_BoneSpace
    alphaBoolBlend {blendOption=Linear}
    boneToModify {boneName=collar_l_back_MSTR}
    rotation {pitch=-15}
   }
   n38 = "Transform (Modify) Bone - Bone: collar_l_back_MSTR"(ComponentPose=@n37, bAlphaBoolEnabled=Get bIsCrouching(), Translation="-1.5,0,1.5", Rotation="-10,0,0", Scale="1,1,1") {
    alphaInputType=Bool
    rotationMode=BMM_Replace
    rotationSpace=BCS_BoneSpace
    translationMode=BMM_Additive
    translationSpace=BCS_BoneSpace
    alphaBoolBlend {blendOption=Linear}
    boneToModify {boneName=collar_l_back_MSTR}
    rotation {pitch=-10}
    translation {x=-1.5, z=1.5}
   }
   n39 = Get bIsCrouchSprinting()
   n40 = "Transform (Modify) Bone - Bone: collar_l_MSTR"!(ComponentPose=@n38, bAlphaBoolEnabled:bool=@n39, Translation="0,0,0", Rotation="-5,0,-10", Scale="1,1,1") -> (Pose) {
    alphaInputType=Bool
    rotationMode=BMM_Replace
    rotationSpace=BCS_BoneSpace
    translationMode=BMM_Replace
    translationSpace=BCS_BoneSpace
    alphaBoolBlend {blendInTime=0.1, blendOption=Linear, blendOutTime=0.1}
    boneToModify {boneName=collar_l_MSTR}
    rotation {pitch=-5, roll=-10}
   }
   n41 = "Transform (Modify) Bone - Bone: collar_r_MSTR"(ComponentPose=@n40, bAlphaBoolEnabled=@n39, Translation="0,0,0", Rotation="-20,0,-10", Scale="1,1,1") {
    alphaInputType=Bool
    rotationMode=BMM_Replace
    rotationSpace=BCS_BoneSpace
    translationMode=BMM_Replace
    translationSpace=BCS_BoneSpace
    alphaBoolBlend {blendInTime=0.1, blendOption=Linear, blendOutTime=0.1}
    boneToModify {boneName=collar_r_MSTR}
    rotation {pitch=-20, roll=-10}
   }
   n42 = "Constraint - dyn_neck_heart_2_mstr (PPPP)"!(
    ComponentPose=@n41
    Alpha=1
    ConstraintWeights_0=0.2641
    ConstraintWeights_1:float=0.1646
    ConstraintWeights_2:float=0.2677
    ConstraintWeights_3:float=0.3037
   ) -> (Pose) {
    alphaBoolBlend {blendOption=Linear}
    boneToModify {boneName=dyn_neck_heart_2_mstr}
    constraintSetup [{targetBone {boneName=spine_05}}, {targetBone {boneName=neck_01}}, {targetBone {boneName=head}}, {targetBone {boneName=neck_02}}]
    constraintWeights [1,1,1,1]
   }
   n43 = "Constraint - dyn_neck_heart_1_mstr (PPPPPP)"!(
    ComponentPose=@n42
    Alpha=1
    ConstraintWeights_0=0.6152
    ConstraintWeights_1:float=0.0142
    ConstraintWeights_2:float=0.0142
    ConstraintWeights_3:float=0.1046
    ConstraintWeights_4:float=0.0921
    ConstraintWeights_5:float=0.1597
   ) -> (Pose) {
    alphaBoolBlend {blendOption=Linear}
    boneToModify {boneName=dyn_neck_heart_1_mstr}
    constraintSetup [
     {targetBone {boneName=spine_05}}
     {targetBone {boneName=clavicle_l}}
     {targetBone {boneName=clavicle_r}}
     {targetBone {boneName=neck_01}}
     {targetBone {boneName=head}}
     {targetBone {boneName=neck_02}}
    ]
    constraintWeights [1,1,1,1,1,1]
   }
   n44 = Get Neck Heart Adjustment() -> (Neck Heart Adjustment:double)
   n45 = "Transform (Modify) Bone - Bone: dyn_neck_heart_1"(ComponentPose=@n43, Alpha=@n44, Translation="0.5,0,0", Rotation="0,0,0", Scale="1,1,1") {
    rotationMode=BMM_Replace
    rotationSpace=BCS_BoneSpace
    translationMode=BMM_Replace
    translationSpace=BCS_BoneSpace
    alphaBoolBlend {blendOption=Linear}
    boneToModify {boneName=dyn_neck_heart_1}
    translation {x=0.5}
   }
   n46 = "Transform (Modify) Bone - Bone: dyn_neck_heart_2"(ComponentPose=@n45, Alpha=@n44, Translation="0.25,0,0", Rotation="0,0,0", Scale="1,1,1") {
    rotationMode=BMM_Replace
    rotationSpace=BCS_BoneSpace
    translationMode=BMM_Replace
    translationSpace=BCS_BoneSpace
    alphaBoolBlend {blendOption=Linear}
    boneToModify {boneName=dyn_neck_heart_2}
    translation {x=0.25}
   }
   n47 = "Constraint - skirt_l_01 (P)"(ComponentPose=@n46, Alpha=@n2, ConstraintWeights_0=1) {
    alphaBoolBlend {blendOption=Linear}
    boneToModify {boneName=skirt_l_01}
    constraintSetup [{targetBone {boneName=thigh_l}}]
    constraintWeights [1]
   }
   n48 = "Constraint - skirt_l_02 (P)"(ComponentPose=@n47, Alpha=@n2, ConstraintWeights_0=1) {
    alphaBoolBlend {blendOption=Linear}
    boneToModify {boneName=skirt_l_02}
    constraintSetup [{targetBone {boneName=thigh_l}}]
    constraintWeights [1]
   }
   n49 = "Constraint - skirt_r_01 (P)"(ComponentPose=@n48, Alpha=@n2, ConstraintWeights_0=1) {
    alphaBoolBlend {blendOption=Linear}
    boneToModify {boneName=skirt_r_01}
    constraintSetup [{targetBone {boneName=thigh_r}}]
    constraintWeights [1]
   }
   n50 = "Constraint - skirt_r_02 (P)"(ComponentPose=@n49, Alpha=@n2, ConstraintWeights_0=1) {
    alphaBoolBlend {blendOption=Linear}
    boneToModify {boneName=skirt_r_02}
    constraintSetup [{targetBone {boneName=thigh_r}}]
    constraintWeights [1]
   }
   n51 = Get SkirtMid FE Alpha_0() -> (SkirtMid FE Alpha_0:double)
   n52 = "Constraint - skirt_01 (PP)"!(ComponentPose=@n50, Alpha=@n51, ConstraintWeights_0=0.5, ConstraintWeights_1:float=0.5) -> (Pose) {
    alphaBoolBlend {blendOption=Linear}
    boneToModify {boneName=skirt_01}
    constraintSetup [{targetBone {boneName=thigh_r}}, {targetBone {boneName=thigh_l}}]
    constraintWeights [1,1]
   }
   n53 = "Constraint - skirt_02 (PP)"!(ComponentPose=@n52, Alpha=@n51, ConstraintWeights_0=0.5, ConstraintWeights_1:float=0.5) -> (Pose) {
    alphaBoolBlend {blendOption=Linear}
    boneToModify {boneName=skirt_02}
    constraintSetup [{targetBone {boneName=thigh_r}}, {targetBone {boneName=thigh_l}}]
    constraintWeights [1,1]
   }
   n54 = "Transform (Modify) Bone - Bone: skirt_MSTR"(ComponentPose=@n53, Alpha=@n2, Translation="0,-2,0", Rotation="0,0,0", Scale="1,1,1") {
    translationMode=BMM_Additive
    translationSpace=BCS_BoneSpace
    alphaBoolBlend {blendOption=Linear}
    boneToModify {boneName=skirt_MSTR}
    translation {y=-2}
   }
   n55 = "Transform (Modify) Bone - Bone: skirt_r_MSTR"(ComponentPose=@n54, Alpha=@n2, Translation="0,-1,-1", Rotation="0,0,0", Scale="1,1,1") {
    translationMode=BMM_Additive
    translationSpace=BCS_BoneSpace
    alphaBoolBlend {blendOption=Linear}
    boneToModify {boneName=skirt_r_MSTR}
    translation {y=-1, z=-1}
   }
   n56 = "Transform (Modify) Bone - Bone: skirt_l_MSTR"(ComponentPose=@n55, Alpha=@n2, Translation="0,-1,1", Rotation="0,0,0", Scale="1,1,1") {
    translationMode=BMM_Additive
    translationSpace=BCS_BoneSpace
    alphaBoolBlend {blendOption=Linear}
    boneToModify {boneName=skirt_l_MSTR}
    translation {y=-1, z=1}
   }
   n57 = Control Rig!(Source=Component To Local(ComponentPose=@n56), Alpha, SKydive Float:double=Get skydiveFloat() -> (skydiveFloat:double)) -> (Pose) {controlRigAssetReference {blueprintRigClass=&/F_MED_Love_Queen_CtrlRig.F_MED_Love_Queen_CtrlRig_C}}
   Output Pose(Result=@n57) {layerGroup=None}
  ]}
 EventGraph {nodes [
   Event BlueprintUpdateAnimation()
   > n1 = Sequence!(execute) -> (then_0, then_1, then_2:Exec, then_3:Exec, then_4:Exec)
   n2 = Get skelMesh() -> (skelMesh:Skeletal Mesh Component Object Reference)
   > Set backCollarAlpha(backCollarAlpha:double=GetAlphaFromReferenceTransformDelta(SkeletalMeshComponent=@n2, BoneName=head, BaseName=neck_01, DeltaComponentType=QuaternionTwist, InputRange=(X=-17,Y=25), TwistAxis="0,0,1")) -> (Output_Get:double)
   > Set thigh_r_up(thigh_r_up:double=GetAlphaFromReferenceTransformDelta(
     SkeletalMeshComponent=@n2
     BoneName=pelvis
     BaseName=thigh_r
     DeltaComponentType=QuaternionTwist
     InputRange=(X=-40,Y=-69)
     TwistAxis="0,0,1"
    )) -> (Output_Get:double)
   > Set Neck Heart Adjustment(Neck Heart Adjustment:double=GetAlphaFromReferenceTransformDelta(SkeletalMeshComponent=@n2, BoneName=neck_01, BaseName=neck_02, DeltaComponentType=QuaternionTwist, InputRange=(X=23,Y=-8), TwistAxis="0,0,1")) -> (Output_Get:double)
   > Set skydiveFloat(skydiveFloat:double="To Float (Boolean)"(InBool=Get bIsSkyDiving())) -> (Output_Get:double)
   n3 = Cast To FortPlayerAnimInstance(execute=@n1.then_3, Object=GetAnimInstance(self=GetOwnerSkeletalMeshComponent()))
   > Set Pistol Alpha(Pistol Alpha:double=1) -> (Output_Get:double)
   > n4 = Switch on EFortWeaponCoreAnimation!(execute, Selection:EFortWeaponCoreAnimation Enum=Get WeaponCoreAnim(self=@n3)) -> (
    Melee:Exec
    Pistol:Exec
    Shotgun:Exec
    PaperBlueprint:Exec
    Rifle:Exec
    MeleeOneHand:Exec
    MachinePistol:Exec
    RocketLauncher:Exec
    GrenadeLauncher:Exec
    GoingCommando:Exec
    AssaultRifle:Exec
    TacticalShotgun:Exec
    SniperRifle:Exec
    TrapPlacement:Exec
    ShoulderLauncher:Exec
    AbilityDecoTool:Exec
    Crossbow:Exec
    C4:Exec
    RemoteControl:Exec
    DualWield:Exec
    AR_BullPup:Exec
    AR_ForwardGrip:Exec
    MedPackPaddles:Exec
    SMG_P90:Exec
    AR_DrumGun:Exec
    Consumable_Small:Exec
    Consumable_Large:Exec
    Balloon:Exec
    MountedTurret:Exec
    CreativeTool:Exec
    ExplosiveBow:Exec
    AshtonIndigo:Exec
    AshtonChicago:Exec
    MeleeDualWield:Exec
    Unarmed:Exec
   )
   Set Pistol Alpha(execute=[@n4.DualWield, @n4.Pistol], Pistol Alpha=0.5)
   n5 = GetOwnerSkeletalMeshComponent()
   n6 = BreakTransform(InTransform=GetDeltaTransformFromRefPose(self=@n5, BoneName=neck_01, BaseName=spine_05))
   Set Look Up Collar Correction(execute=@n1.then_2, Look Up Collar Correction:double=MapRangeClamped(Value=BreakRotator(InRot=@n6.Rotation).Yaw, InRangeA=-15, InRangeB=-50)) -> (Output_Get:double)
   n7 = BreakTransform(InTransform=GetDeltaTransformFromRefPose(self=@n5, BoneName=clavicle_r, BaseName=upperarm_r))
   > Set Rifle ADS Down Correction(Rifle ADS Down Correction:double=MapRangeClamped(Value=BreakRotator(InRot=@n7.Rotation).Yaw, InRangeA=-44, InRangeB=-49)) -> (Output_Get:double)
   n8 = Branch(execute=@n1.then_1, Condition=Get bIsSkyDiving())
   > n9 = Branch(Condition=Get bIsParachuteOpened())
   Set gravityOverride(execute=[@n9.then, @n8.else], gravityOverride:Vector="0, 0,-980") -> (Output_Get:Vector)
   Set gravityOverride(execute=@n9.else, gravityOverride="0, 0,-100")
   n10 = Get bIsPlayingEmote()
   n11 = Branch(execute=@n1.then_4, Condition=AND Boolean(A=Get bIsInFrontEnd(), B=NOT Boolean(A=@n10)))
   > Set Skirt FE Alpha(Skirt FE Alpha:double=1) -> (Output_Get:double)
   > Set SkirtMid FE Alpha_0(SkirtMid FE Alpha_0:double=0.5) -> (Output_Get:double)
   Set Skirt FE Alpha(execute=@n11.else, Skirt FE Alpha=0)
   > Set SkirtMid FE Alpha_0(SkirtMid FE Alpha_0=0)
   Event BlueprintInitializeAnimation()
   > Set skelMesh(skelMesh:Skeletal Mesh Component Object Reference=GetOwnerSkeletalMeshComponent()) -> (Output_Get:Skeletal Mesh Component Object Reference)
   PrintString!(execute, InString="To String (Float)"(InDouble=0), bPrintToScreen, bPrintToLog, TextColor, Duration, Key) -> (then)
  ]}
]
variables [
 backCollarAlpha {type=number}
 gravityOverride {type=Vector, default=<0 0 -980>}
 Is Sky Diving {type=boolean}
 Look Up Collar Correction {type=number}
 Neck Heart Adjustment {type=number}
 Pistol Alpha {type=number}
 Rifle ADS Down Correction {type=number}
 skelMesh {type=SkeletalMeshComponent}
 Skirt FE Alpha {type=number}
 SkirtMid FE Alpha_0 {type=number}
 skydiveFloat {type=number}
 thigh_r_up {type=number}
]
