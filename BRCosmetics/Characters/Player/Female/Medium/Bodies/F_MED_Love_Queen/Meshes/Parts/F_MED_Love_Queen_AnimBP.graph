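# Animation Blueprint for a Love Queen female body part that adjusts hair bones as the head tilts during targeting.
# It copies the attached parent's pose, then additively translates and scales dyn_hair_r and dyn_hair_center, weighted by ADS Head Up.
# Each update, that weight is mapped from head-to-spine_05 delta yaw while targeting and zeroed otherwise.
# The dyn_hair_l modifier is unconnected.
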AnimBlueprint /BRCosmetics/Characters/Player/Female/Medium/Bodies/F_MED_Love_Queen/Meshes/Parts/F_MED_Love_Queen_AnimBP
target_skeleton=&/F_MED_Love_Queen_Skeleton
graphs [
 AnimGraph {nodes [
   "Transform (Modify) Bone - Bone: dyn_hair_l"(Alpha=0.8, Translation="0,0,0", Rotation="0,0,0", Scale="1,1,1") {
    rotationMode=BMM_Additive
    rotationSpace=BCS_BoneSpace
    scaleMode=BMM_Additive
    scaleSpace=BCS_BoneSpace
    translationMode=BMM_Additive
    translationSpace=BCS_BoneSpace
    alphaBoolBlend {blendOption=Linear}
    boneToModify {boneName=dyn_hair_l}
   }
   n1 = Get ADS Head Up() -> (ADS Head Up:double)
   n2 = "Transform (Modify) Bone - Bone: dyn_hair_r"(
    ComponentPose=Local To Component(LocalPose=Copy Pose From Mesh() {bUseAttachedParent=true})
    Alpha=@n1
    Translation="0,1,0"
    Rotation="0,0,0"
    Scale="0.6,1.000,0.8"
   ) {
    rotationMode=BMM_Additive
    rotationSpace=BCS_BoneSpace
    scaleMode=BMM_Additive
    scaleSpace=BCS_BoneSpace
    translationMode=BMM_Additive
    translationSpace=BCS_BoneSpace
    alphaBoolBlend {blendOption=Linear}
    boneToModify {boneName=dyn_hair_r}
    scale {x=0.6, z=0.8}
    translation {y=1}
   }
   n3 = "Transform (Modify) Bone - Bone: dyn_hair_center"(ComponentPose=@n2, Alpha=@n1, Translation="0.3,0.000,0.000", Rotation="0,0,0", Scale="0.75,1.000,1.000") {
    rotationMode=BMM_Additive
    rotationSpace=BCS_BoneSpace
    scaleMode=BMM_Additive
    scaleSpace=BCS_BoneSpace
    translationMode=BMM_Additive
    translationSpace=BCS_BoneSpace
    alphaBoolBlend {blendOption=Linear}
    boneToModify {boneName=dyn_hair_center}
    scale {x=0.75}
    translation {x=0.3}
   }
   Output Pose(Result=Component To Local(ComponentPose=@n3)) {layerGroup=None}
  ]}
 EventGraph {nodes [
   Event BlueprintUpdateAnimation()
   > Sequence()
   > n1 = Branch(Condition=Get bIsTargeting())
   n2 = BreakTransform(InTransform=GetDeltaTransformFromRefPose(self=GetOwnerSkeletalMeshComponent(), BoneName=head, BaseName=spine_05))
   > Set ADS Head Up(ADS Head Up:double=MapRangeClamped(Value=BreakRotator(InRot=@n2.Rotation).Yaw, InRangeA=-3, InRangeB=-40)) -> (Output_Get:double)
   Set ADS Head Up(execute=@n1.else, ADS Head Up=0)
  ]}
]
variables [
 ADS Head Up {type=number}
 backCollarAlpha {type=number}
 skelMesh {type=SkeletalMeshComponent}
]
